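# Sound cue for the charge-start sound of Milo's secondary launcher weapon in the Ashton limited-time mode.
# A single wave player plays one charge-start sample at reduced volume through a third-person ranged-weapon fire sound class.
# It uses close-projectile weapon attenuation, with natural-sound falloff overrides also stored on the cue.
# Its concurrency override settings specify a single instance that blocks new plays rather than cutting off the current one.
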SoundCue /Game/Athena/Sounds/LTMs/Ashton/Milo/Ashton_Milo_Secondary_Charge_Cue
root=WavePlayer_3
nodes [{id=WavePlayer_3, properties {soundWaveAssetPtr=&/Ashton_Milo_Launcher_Charge_Start_03}}]
properties {
 attenuationSettings=&Fort_Weapon_Projectile_Close_Athena_Att
 maxDistance=4950
 soundClassObject=&Fort_WeaponRanged_Fire_3P_PSM
 volumeMultiplier=0.4
 attenuationOverrides {distanceAlgorithm=NaturalSound, falloffDistance=4000, attenuationShapeExtents {x=1000}}
 concurrencyOverrides {maxCount=1, resolutionRule=PreventNew}
}
