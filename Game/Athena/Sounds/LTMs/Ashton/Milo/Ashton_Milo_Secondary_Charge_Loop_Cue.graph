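# Sound cue for the looping secondary charge sound of the Milo launcher in the Ashton LTM.
# A looping wave feeds an enveloper that ramps volume up from silence in under a second, rising more steeply toward the end.
# That passes through a small-rifle attenuation node and a root sound class node that overrides to third-person ranged weapon fire.
# The cue plays at reduced volume, with concurrency overridden to block new instances at the limit.
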SoundCue /Game/Athena/Sounds/LTMs/Ashton/Milo/Ashton_Milo_Secondary_Charge_Loop_Cue
root=SoundClass_25
nodes [
 {id=Attenuation_35, children [Enveloper_0], properties {attenuationSettings=/Game/Sounds/Attenuation/Fort_Weapon_Rifle_Sml_Att}}
 {
  id=Enveloper_0
  children [WavePlayer_1]
  properties {pitchCurve {editorCurveData {keys [{time=0}]}}, volumeCurve {editorCurveData {keys [<at 0 0>,<at 0.5218 0.2631>,<at 0.7623 1>]}}}
 }
 {
  id=SoundClass_25
  children [Attenuation_35]
  properties {soundClassOverride=&Fort_WeaponRanged_Fire_3P_PSM}
 }
 {id=WavePlayer_1, properties {bLooping=true, soundWaveAssetPtr=&/Ashton_Milo_Launcher_Charge_Loop_02}}
]
properties {
 attenuationSettings=/Game/Sounds/Attenuation/Fort_Weapon_IronSights_Att
 bOverrideConcurrency=true
 maxDistance=3000
 soundClassObject=&Fort_Gadgets_Items
 volumeMultiplier=0.2
 attenuationOverrides {distanceAlgorithm=NaturalSound, falloffDistance=4000, attenuationShapeExtents {x=1000}}
 concurrencyOverrides {resolutionRule=PreventNew}
}
